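; Controller component for the Ad Hoc Squads feature that gives the player text and audio feedback on squad invite, join, failed join and leave events.
; Each event formats text for ShowUIMessage, a temporary debug print standing in for UI messaging, with failure reasons mapped from an enum.
; All but failed joins also call PlaySFX, which, unless audio is disabled, plays stingers and 2D or pawn-attached sounds, selecting 1P/3P by local control.
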(event EventOnPlayerInitiatedSquadInvite
  (params (PlayerInitiatingInvite :type Fort%20Player%20State%20Athena%20Object%20Reference))
  ;[0.1194 1 0] Player Initiated Squad Invite
  (Utilities|IsValid PlayerInitiatingInvite
    (:Is%20Valid
      (CallFunction|ShowUIMessage (Utilities|Text|FormatText "NSLOCTEXT(\"[0317FAC04E4AF9C834DE0EA15F4C244A]\", \"ADE45D5A4111F0589D41869555EB96AE\", \"{Player} Initated squad invite. Squad size is {NumberOfPlayers}.\")" PlayerInitiatingInvite (CallFunction|GetNumberOfPlayersinSquad PlayerInitiatingInvite)))
      (CallFunction|PlaySFX 0 true (PlayerState|GetCurrentPawn PlayerInitiatingInvite) (Variables|Sound|GetSFX_Whistle_3P) (Variables|Sound|GetSFX_Whistle_1P)))))

(event EventOnOwningPlayerJoinedSquad
  (params (PlayerTheOwningPlayerIsJoining :type Fort%20Player%20State%20Athena%20Object%20Reference) (bInviteInstigatedByThisPlayer :type Boolean))
  ;[0.2255 0.2209 1] Owning Player Joined Squad
  (Utilities|IsValid PlayerTheOwningPlayerIsJoining
    (:Is%20Valid
      (CallFunction|ShowUIMessage (Utilities|Text|FormatText "NSLOCTEXT(\"[0317FAC04E4AF9C834DE0EA15F4C244A]\", \"5143555D4DE9CBC69FD39CA009A8B88D\", \"You joined {Player}\\'s squad.\")" (PlayerState|GetPlayerNameforStreaming PlayerTheOwningPlayerIsJoining)))
      (CallFunction|PlaySFX (Variables|Sound|GetMusicStinger_JoinSquad) false 0 (Variables|Sound|GetSFX_UI_JoinSquad) (Variables|Sound|GetSFX_UI_JoinSquad)))))

(event EventOnOtherPlayerJoinedSquad
  (params (JoiningPlayer :type Fort%20Player%20State%20Athena%20Object%20Reference) (bInviteInstigatedByThisPlayer :type Boolean))
  ;[0.7109 0.2222 1] Other Player Joined Squad
  (Utilities|IsValid JoiningPlayer
    (:Is%20Valid
      (CallFunction|ShowUIMessage (Utilities|Text|FormatText "NSLOCTEXT(\"[0317FAC04E4AF9C834DE0EA15F4C244A]\", \"8A8614DA46C31D18188A9DA9E8E110F3\", \"{Player} joined your squad\")" (PlayerState|GetPlayerNameforStreaming JoiningPlayer)))
      (CallFunction|PlaySFX (Variables|Sound|GetMusicStinger_JoinSquad) false 0 (Variables|Sound|GetSFX_UI_JoinSquad) (Variables|Sound|GetSFX_UI_JoinSquad)))))

(event EventOnFailedToJoinSquad
  (params (PlayerFromSquad :type Fort%20Player%20State%20Athena%20Object%20Reference) (FailureReason :type EAdHocSquads_SquadUpResult%20Enum :default "Success"))
  (bind _returnvalue (Utilities|Select :Index FailureReason :Success "NSLOCTEXT(\"[0317FAC04E4AF9C834DE0EA15F4C244A]\", \"024261C54E09DE53C58D5B8D5F91ADDF\", \"Error 100\")" :Failure_OneOrMorePlayersWereNull "NSLOCTEXT(\"[0317FAC04E4AF9C834DE0EA15F4C244A]\", \"9342E10345F91981F3ECA78B17A788C2\", \"Error 101\")" :Failure_BothPlayersAreInAdHocSquads "NSLOCTEXT(\"[0317FAC04E4AF9C834DE0EA15F4C244A]\", \"47777F6742CAF8F869A354A43753678F\", \"You are already in a squad\")" :Failure_BothPlayersAreAlreadyInTheSameAdHocSquad "NSLOCTEXT(\"[0317FAC04E4AF9C834DE0EA15F4C244A]\", \"3A703B764A90ECC28E5E6CAFC1B39895\", \"You are already in the same squad\")" :Failure_CouldNotCreateAdHocSquad "NSLOCTEXT(\"[0317FAC04E4AF9C834DE0EA15F4C244A]\", \"A85D290F4032D06FEFFBC7ACB7C25267\", \"Error 102\")" :Failure_CalledOnClient "NSLOCTEXT(\"[0317FAC04E4AF9C834DE0EA15F4C244A]\", \"296D228F483B23FC3CC222B950CE7CA0\", \"Error 103\")" :Failure_SquadIsAlreadyFull "NSLOCTEXT(\"[0317FAC04E4AF9C834DE0EA15F4C244A]\", \"D804548C4FC7E5D4AE7D9A82B714F0A3\", \"The squad you tried to join is already full\")" :Failure_MutatorIsDisabled "NSLOCTEXT(\"[0317FAC04E4AF9C834DE0EA15F4C244A]\", \"240655A5403035F96936D3995B41ABE5\", \"Error 104\")" :Failure_Unknown "NSLOCTEXT(\"[0317FAC04E4AF9C834DE0EA15F4C244A]\", \"DADA1A274AAFFBFAE2D731A98B2AA078\", \"Error 999\")"))
  ;[1 0.5417 0.5417] Player Failed to Join Squad
  (Utilities|IsValid PlayerFromSquad
    (:Is%20Valid
      (CallFunction|ShowUIMessage (Utilities|Text|FormatText "NSLOCTEXT(\"[0317FAC04E4AF9C834DE0EA15F4C244A]\", \"FFCE071E4EE37FFF69A64FBD5517B9F2\", \"[You failed to join {Player}\\'s squad: {Reason}.\")" (PlayerState|GetPlayerNameforStreaming PlayerFromSquad) _returnvalue)))
    (:Is%20Not%20Valid
      (CallFunction|ShowUIMessage (Utilities|Text|FormatText "NSLOCTEXT(\"[0317FAC04E4AF9C834DE0EA15F4C244A]\", \"4CE84D2D4351CD853BD489A8C6357FE1\", \"You failed to join UNKNOWN PLAYER\\'s squad: {Reason}.\")" _returnvalue)))))

(event EventOnOwningPlayerLeftSquad
  (params (LeaveReason :type EAdHocSquads_LeaveSquadReason%20Enum :default "ManualLeave"))
  ;[1 0.6047 0.09861] Owning Player Left Squad
  (CallFunction|ShowUIMessage (Utilities|Text|FormatText "NSLOCTEXT(\"[0317FAC04E4AF9C834DE0EA15F4C244A]\", \"02FFA6D349041C85E51CFB8A91F8306D\", \"You have left your squad.\")"))
  (CallFunction|PlaySFX (Variables|Sound|GetMusicStinger_LeaveSquad) false 0 (Variables|Sound|GetSFX_UI_LeaveSquad) (Variables|Sound|GetSFX_UI_LeaveSquad)))

(event EventOnOtherPlayerLeftSquad
  (params (LeavingPlayer :type Fort%20Player%20State%20Athena%20Object%20Reference) (LeaveReason :type EAdHocSquads_LeaveSquadReason%20Enum :default "ManualLeave"))
  ;[0.2604 0.1482 0.01356] Other Player Left Squad
  (Utilities|IsValid LeavingPlayer
    (:Is%20Valid
      (CallFunction|ShowUIMessage (Utilities|Text|FormatText "NSLOCTEXT(\"[0317FAC04E4AF9C834DE0EA15F4C244A]\", \"70097E894D1C8E261D91F5A57D100F49\", \"{Player} left your squad.\")" (PlayerState|GetPlayerNameforStreaming LeavingPlayer)))
      (CallFunction|PlaySFX (Variables|Sound|GetMusicStinger_LeaveSquad) false 0 (Variables|Sound|GetSFX_UI_LeaveSquad) (Variables|Sound|GetSFX_UI_LeaveSquad)))))

(fn PlaySFX
  (params (Music%20Stinger :type Sound%20Base%20Object%20Reference) (SFX%20is%20Attenuated :type Boolean) (Attenuated%20SFX%20Attached%20to%20Fort%20Player%20Pawn :type Fort%20Player%20Pawn%20Object%20Reference) (UI%20SFX%20/%20Other%20SFX_3P :type Sound%20Base%20Object%20Reference) (UI%20SFX%20/%20Other%20SFX_1P :type Sound%20Base%20Object%20Reference))
  (bind _blocallycontrolled (LocalVariables|GetbLocallyControlled))
  (bind _ui_sfx___other_sfx_1p (Variables|GetUI%20SFX%20/%20Other%20SFX_1P))
  (bind _ui_sfx___other_sfx_3p (Variables|GetUI%20SFX%20/%20Other%20SFX_3P))
  (bind _returnvalue (select _blocallycontrolled _ui_sfx___other_sfx_1p _ui_sfx___other_sfx_3p))
  (Variables|Sound|GetbDisableAudio
    (:then)
    (:else
      (Audio|PlaySound2D Music%20Stinger)
      (Utilities|FlowControl|Sequence
        (:then_0
          (bind _attenuated_sfx_attached_to_fort_player_pawn (Variables|GetAttenuated%20SFX%20Attached%20to%20Fort%20Player%20Pawn)
            (:then
              (LocalVariables|SetbLocallyControlled (Pawn|IsLocallyControlled _attenuated_sfx_attached_to_fort_player_pawn)))
            (:else)))
        (:then_1
          (if SFX%20is%20Attenuated
            (bind _attenuated_sfx_attached_to_fort_player_pawn_2 (Variables|GetAttenuated%20SFX%20Attached%20to%20Fort%20Player%20Pawn)
              (:then
                (Audio|SpawnSoundAttached _returnvalue (Variables|/Script/Engine.Actor|GetRootComponent _attenuated_sfx_attached_to_fort_player_pawn_2)))
              (:else))
            (else
              (Audio|PlaySound2D _returnvalue))))))))

(fn ShowUIMessage
  (params (DisplayText :type Text))
  ;[0.5052 0.5052 0.5052] Temp logging code. To be replaced with UI messaging to the local player (https://jira.it.epicgames.com/browse/FORT-345078).
  (Networking|CanExecuteCosmeticEvents
    (:True
      (Development|PrintString (Utilities|String|ToString(Text) DisplayText) (Variables|Default|GetbDebugPrintToScreen) (Variables|Default|GetbDebugPrintToLog)))))
