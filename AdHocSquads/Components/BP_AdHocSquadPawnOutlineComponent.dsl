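; Ad hoc squad pawn outline component that highlights its owning PlayerPawn_Athena through the PawnHighlight system.
; On BeginPlay it calls UpdateOutline, which initializes the outline if not yet showing, then updates its color.
; Initialization caches the owner as Target Pawn, printing an error if the cast fails.
; Color updates apply a highlight from the base class's scaled OutlineColor with the component as source, which cleanup uses to remove it.
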(event EventBeginPlay
  (|CallParentFunction)
  (CallFunction|UpdateOutline))

(event Custom|UpdateOutline
  (if (not (Variables|Default|GetbShowing))
    (CallFunction|InitializeOutline)
    (Variables|Default|SetbShowing true))
  (CallFunction|UpdateColor))

(event EventCleanUpOutlinePerformance
  (Utilities|IsValid (Variables|Default|GetTarget%20Pawn)
    (:Is%20Valid
      (PawnHighlight|RemovePawnHighlightFromSource (Variables|Default|GetTarget%20Pawn) self))))

(event EventInitializeOutlinePerformance
  (bind _asplayer_pawn_athena (Utilities|Casting|CastToPlayerPawn_Athena (Components|GetOwner))
    (:then
      (Variables|Default|SetTarget%20Pawn _asplayer_pawn_athena))
    (:CastFailed
      (Development|PrintString "Unable to get PlayerPawn_Athena"))))

(event EventUpdateColorPerformance
  (Utilities|IsValid (Variables|Default|GetTarget%20Pawn)
    (:Is%20Valid
      (PawnHighlight|HighlightPawnFromSource (Variables|Default|GetTarget%20Pawn) self (Utilities|Struct|MakePawnHighlight 1.0 "(R=0.000000,G=0.000000,B=0.000000,A=0.000000)" (* (Variables|PawnOutlineComponentBase|GetOutlineColor) (Variables|PawnOutlineComponentBase|GetOutlineColorScale)) 5.0 15.0 0.0) false))))
